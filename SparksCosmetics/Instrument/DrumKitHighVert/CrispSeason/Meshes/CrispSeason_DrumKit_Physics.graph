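# Physics asset for the CrispSeason drum kit cosmetic, covering the two crash cymbals and the ride with their stands.
# Capsule bodies sit on a handle, three stands and two chained dyn_brush bones per cymbal.
# Constraints lock the crash stands' rotation to the handle, while the ride stand has no such link.
# Each first brush bone attaches to its stand with small swing limit angles and locked twist, then links to a twist-locked second bone.
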PhysicsAsset /SparksCosmetics/Instrument/DrumKitHighVert/CrispSeason/Meshes/CrispSeason_DrumKit_Physics
bodies {
 crash_stand_l [{
   shapeType=Capsule
   length=55.1622
   radius=2
   shapeName=crash_stand_l_capsule
   center=<51.952 0.3609 0.7573>
   rotation=<rot 0 90.398 -89.1649>
  }]
 crash_stand_r [{shapeType=Capsule, length=57.098, radius=2, shapeName=crash_stand_r_capsule, center=<55.9572 0 0>, rotation=<rot 0 -90.6937 -88.3339>}]
 dyn_brush_crash_cymbal_l [{shapeType=Capsule, length=2, radius=4, shapeName=dyn_brush_crash_cymbal_l_capsule, rotation=<rot 35.2735 -90 -4.4758>}]
 dyn_brush_crash_cymbal_l_02 [{
   shapeType=Capsule
   length=5.1952
   radius=3.9936
   shapeName=dyn_brush_crash_cymbal_l_02_capsule
   center=<0.0198 0.0003 1.3344>
   rotation=<rot 0.1728 -90 38.982>
  }]
 dyn_brush_crash_cymbal_r [{shapeType=Capsule, length=3, radius=4, shapeName=dyn_brush_crash_cymbal_r_capsule}]
 dyn_brush_crash_cymbal_r_02 [{
   shapeType=Capsule
   length=4.1899
   radius=3.2082
   shapeName=dyn_brush_crash_cymbal_r_02_capsule
   center=<2.3474 0.3729 1.5598>
   rotation=<rot 9.6363 -90 31.2504>
  }]
 dyn_brush_ride [{shapeType=Capsule, length=2, radius=4, shapeName=dyn_brush_ride_capsule}]
 dyn_brush_ride_02 [{
   shapeType=Capsule
   length=4.325
   radius=2.5975
   shapeName=dyn_brush_ride_02_capsule
   center=<1.8553 0.5896 -0.2162>
   rotation=<rot 0 -35.0778 -76.3778>
  }]
 handle [{shapeType=Capsule, length=0.505, radius=0.505, rotation=<rot 0 -90 0>}]
 ride_stand [{shapeType=Capsule, length=42.8085, radius=2, shapeName=ride_stand_capsule, center=<42.7388 0 0>, rotation=<rot 0 90.84 -89.1418>}]
}
constraints [
 {
  bone1Name=crash_stand_l
  bone2Name=handle
  swing1LimitDegrees=45
  swing1Motion=Locked
  swing2LimitDegrees=45
  swing2Motion=Locked
  twistLimitDegrees=45
  twistMotion=Locked
 }
 {
  bone1Name=dyn_brush_crash_cymbal_l
  bone2Name=crash_stand_l
  swing1LimitDegrees=4
  swing2LimitDegrees=4
  twistLimitDegrees=0
  twistMotion=Locked
 }
 {
  bone1Name=dyn_brush_crash_cymbal_l_02
  bone2Name=dyn_brush_crash_cymbal_l
  swing1LimitDegrees=45
  swing2LimitDegrees=45
  twistLimitDegrees=45
  twistMotion=Locked
 }
 {
  bone1Name=dyn_brush_ride_02
  bone2Name=dyn_brush_ride
  swing1LimitDegrees=45
  swing2LimitDegrees=45
  twistLimitDegrees=45
  twistMotion=Locked
 }
 {bone1Name=dyn_brush_ride, bone2Name=ride_stand, swing1LimitDegrees=4, swing2LimitDegrees=4, twistLimitDegrees=45, twistMotion=Locked}
 {
  bone1Name=dyn_brush_crash_cymbal_r
  bone2Name=crash_stand_r
  swing1LimitDegrees=4
  swing2LimitDegrees=4
  twistLimitDegrees=45
  twistMotion=Locked
 }
 {
  bone1Name=dyn_brush_crash_cymbal_r_02
  bone2Name=dyn_brush_crash_cymbal_r
  swing1LimitDegrees=45
  swing2LimitDegrees=45
  twistLimitDegrees=45
  twistMotion=Locked
 }
 {
  bone1Name=crash_stand_r
  bone2Name=handle
  swing1LimitDegrees=45
  swing1Motion=Locked
  swing2LimitDegrees=45
  swing2Motion=Locked
  twistLimitDegrees=45
  twistMotion=Locked
 }
]
