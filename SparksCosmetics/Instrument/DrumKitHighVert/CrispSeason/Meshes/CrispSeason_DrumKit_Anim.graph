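# Post-process style Anim Blueprint for the CrispSeason drum kit cosmetic, adding procedural secondary motion on top of its input pose.
# A RigidBody node simulates the kit's physics asset in the space of the handle bone.
# Four chained, identically tuned Anim Dynamics nodes then swing the snare, both tom and floor tom hinges on angular springs with one-sided, single-axis limits.
# The event graph holds no logic.
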AnimBlueprint /SparksCosmetics/Instrument/DrumKitHighVert/CrispSeason/Meshes/CrispSeason_DrumKit_Anim
target_skeleton=&/CrispSeason_DrumKit_Skeleton
graphs [
 AnimGraph {nodes [
   n1 = RigidBody(ComponentPose=Local To Component(LocalPose=Input Pose())) {
    bOverrideWorldGravity=true
    overridePhysicsAsset=&/CrispSeason_DrumKit_Physics
    simulationSpace=BaseBoneSpace
    alphaBoolBlend {blendOption=Linear}
    baseBoneRef {boneName=handle}
    componentAppliedLinearAccClamp=<600 600 600>
    componentLinearAccScale=<0.4 0.4 0.4>
    componentLinearVelScale=<0.4 0.4 0.4>
    overrideWorldGravity {z=-980}
    simSpaceSettings {maxAngularAcceleration=100, maxAngularVelocity=10, maxLinearAcceleration=400, maxLinearVelocity=400, velocityScaleZ=0.2, worldAlpha=1}
   }
   n2 = Anim Dynamics - Bone: dyn_snare_hinge(ComponentPose=@n1, Alpha=1, ExternalForce="0,0,0") {
    angularSpringConstant=1500
    bAngularSpring=true
    bUseGravityOverride=true
    simulationSpace=RootRelative
    boundBone {boneName=dyn_snare_hinge}
    componentAppliedLinearAccClamp=<1000 1000 1000>
    componentLinearAccScale=<0.7 0.7 0.7>
    componentLinearVelScale=<0.2 0.2 0.2>
    gravityOverride {z=-980}
    physicsBodyDefinitions [{
      boundBone {boneName=dyn_snare_hinge}
      boxExtents=<12 12 12>
      constraintSetup {angularLimitsMax=<0 0 0>, angularLimitsMin=<0 -35 0>, angularTarget=<0 1 0>}
      localJointOffset=<3 0 6>
     }]
    relativeSpaceBone {boneName=handle}
   }
   n3 = Anim Dynamics - Bone: dyn_tom_r_hinge(ComponentPose=@n2, Alpha=1, ExternalForce="0,0,0") {
    angularSpringConstant=1500
    bAngularSpring=true
    bUseGravityOverride=true
    simulationSpace=RootRelative
    boundBone {boneName=dyn_tom_r_hinge}
    componentAppliedLinearAccClamp=<1000 1000 1000>
    componentLinearAccScale=<0.7 0.7 0.7>
    componentLinearVelScale=<0.2 0.2 0.2>
    gravityOverride {z=-980}
    physicsBodyDefinitions [{
      boundBone {boneName=dyn_tom_r_hinge}
      boxExtents=<12 12 12>
      constraintSetup {angularLimitsMax=<0 0 0>, angularLimitsMin=<0 -35 0>, angularTarget=<0 1 0>}
      localJointOffset=<3 0 6>
     }]
    relativeSpaceBone {boneName=handle}
   }
   n4 = Anim Dynamics - Bone: dyn_tom_l_hinge(ComponentPose=@n3, Alpha=1, ExternalForce="0,0,0") {
    angularSpringConstant=1500
    bAngularSpring=true
    bUseGravityOverride=true
    simulationSpace=RootRelative
    boundBone {boneName=dyn_tom_l_hinge}
    componentAppliedLinearAccClamp=<1000 1000 1000>
    componentLinearAccScale=<0.7 0.7 0.7>
    componentLinearVelScale=<0.2 0.2 0.2>
    gravityOverride {z=-980}
    physicsBodyDefinitions [{
      boundBone {boneName=dyn_tom_l_hinge}
      boxExtents=<12 12 12>
      constraintSetup {angularLimitsMax=<0 0 0>, angularLimitsMin=<0 -35 0>, angularTarget=<0 1 0>}
      localJointOffset=<3 0 6>
     }]
    relativeSpaceBone {boneName=handle}
   }
   n5 = Anim Dynamics - Bone: dyn_floortom_hinge(ComponentPose=@n4, Alpha=1, ExternalForce="0,0,0") {
    angularSpringConstant=1500
    bAngularSpring=true
    bUseGravityOverride=true
    simulationSpace=RootRelative
    boundBone {boneName=dyn_floortom_hinge}
    componentAppliedLinearAccClamp=<1000 1000 1000>
    componentLinearAccScale=<0.7 0.7 0.7>
    componentLinearVelScale=<0.2 0.2 0.2>
    gravityOverride {z=-980}
    physicsBodyDefinitions [{
      boundBone {boneName=dyn_floortom_hinge}
      boxExtents=<12 12 12>
      constraintSetup {angularLimitsMax=<0 0 0>, angularLimitsMin=<0 -35 0>, angularTarget=<0 1 0>}
      localJointOffset=<3 0 6>
     }]
    relativeSpaceBone {boneName=handle}
   }
   Output Pose(Result=Component To Local(ComponentPose=@n5))
  ]}
 EventGraph {nodes [Event BlueprintUpdateAnimation()]}
]
variables []
